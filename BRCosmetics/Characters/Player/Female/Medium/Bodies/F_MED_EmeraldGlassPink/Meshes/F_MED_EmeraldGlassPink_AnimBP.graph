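# Post-process style animation Blueprint for the female medium EmeraldGlassPink body, adding cosmetic deformation to a pose copied from the parent mesh.
# The pose runs through a deform Control Rig fed a glute transform, then weighted constraints driving hip geo bones from pelvis, spine and leg bones.
# A fixed additive transform adjusts the right kneepad.
# The event graph records knee-bend yaw while crouching or sliding, but the AnimGraph shown never reads it.
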AnimBlueprint /BRCosmetics/Characters/Player/Female/Medium/Bodies/F_MED_EmeraldGlassPink/Meshes/F_MED_EmeraldGlassPink_AnimBP
target_skeleton=&/F_MED_EmeraldGlassPink_Skeleton
graphs [
 AnimGraph {nodes [
   Control Rig() {
    alphaBoolBlend {blendOption=Linear}
    controlRigAssetReference {blueprintRigClass=&/F_MED_DeformRig_EmeraldGlassPink.F_MED_DeformRig_EmeraldGlassPink_C}
   }
   n1 = Control Rig!(Source=Copy Pose From Mesh() {bUseAttachedParent=true}, Alpha, GluteTransform:Transform=Get Glute Transform() -> (Glute Transform:Transform)) -> (Pose) {
    alphaBoolBlend {blendOption=Linear}
    controlRigAssetReference {blueprintRigClass=&F_MED_DeformRig_EmeraldGlassPink_C}
   }
   n2 = "Constraint - deform_hip_geo_base_a_l (PPPP)"!(
    ComponentPose=Local To Component(LocalPose=@n1)
    Alpha=1
    ConstraintWeights_0=0.7893
    ConstraintWeights_1:float=0.1106
    ConstraintWeights_2:float=0.0811
    ConstraintWeights_3:float=0.019
   ) -> (Pose) {
    alphaBoolBlend {blendOption=Linear}
    boneToModify {boneName=deform_hip_geo_base_a_l}
    constraintSetup [
     {targetBone {boneName=pelvis}}
     {targetBone {boneName=thigh_twist_01_l}}
     {targetBone {boneName=deform_groin_l}}
     {targetBone {boneName=spine_02}}
    ]
    constraintWeights [1,1,1,1]
   }
   n3 = "Constraint - deform_hip_geo_base_b_l (PPPPP)"!(
    ComponentPose=@n2
    Alpha=1
    ConstraintWeights_0=0.8966
    ConstraintWeights_1:float=0.0684
    ConstraintWeights_2:float=0.0188
    ConstraintWeights_3:float=0.0103
    ConstraintWeights_4:float=0.0059
   ) -> (Pose) {
    alphaBoolBlend {blendOption=Linear}
    boneToModify {boneName=deform_hip_geo_base_b_l}
    constraintSetup [
     {targetBone {boneName=pelvis}}
     {targetBone {boneName=spine_02}}
     {targetBone {boneName=thigh_twist_01_l}}
     {targetBone {boneName=spine_01}}
     {targetBone {boneName=deform_glute_l}}
    ]
    constraintWeights [1,1,1,1,1]
   }
   n4 = "Constraint - deform_hip_geo_base_c_l (PPPPP)"!(
    ComponentPose=@n3
    Alpha=1
    ConstraintWeights_0=0.7491
    ConstraintWeights_1:float=0.159
    ConstraintWeights_2:float=0.0648
    ConstraintWeights_3:float=0.0138
    ConstraintWeights_4:float=0.0133
   ) -> (Pose) {
    alphaBoolBlend {blendOption=Linear}
    boneToModify {boneName=deform_hip_geo_base_c_l}
    constraintSetup [
     {targetBone {boneName=pelvis}}
     {targetBone {boneName=spine_02}}
     {targetBone {boneName=spine_01}}
     {targetBone {boneName=spine_04}}
     {targetBone {boneName=spine_03}}
    ]
    constraintWeights [1,1,1,1,1]
   }
   n5 = "Constraint - deform_hip_geo_base_mid (PPPPP)"!(
    ComponentPose=@n4
    Alpha=1
    ConstraintWeights_0=0.598
    ConstraintWeights_1:float=0.2455
    ConstraintWeights_2:float=0.1042
    ConstraintWeights_3:float=0.0292
    ConstraintWeights_4:float=0.023
   ) -> (Pose) {
    alphaBoolBlend {blendOption=Linear}
    boneToModify {boneName=deform_hip_geo_base_mid}
    constraintSetup [
     {targetBone {boneName=pelvis}}
     {targetBone {boneName=spine_02}}
     {targetBone {boneName=spine_01}}
     {targetBone {boneName=spine_03}}
     {targetBone {boneName=spine_04}}
    ]
    constraintWeights [1,1,1,1,1]
   }
   n6 = "Constraint - deform_hip_geo_base_a_r (PPPP)"!(
    ComponentPose=@n5
    Alpha=1
    ConstraintWeights_0=0.7893
    ConstraintWeights_1:float=0.1106
    ConstraintWeights_2:float=0.0811
    ConstraintWeights_3:float=0.019
   ) -> (Pose) {
    alphaBoolBlend {blendOption=Linear}
    boneToModify {boneName=deform_hip_geo_base_a_r}
    constraintSetup [
     {targetBone {boneName=pelvis}}
     {targetBone {boneName=thigh_twist_01_r}}
     {targetBone {boneName=deform_groin_r}}
     {targetBone {boneName=spine_02}}
    ]
    constraintWeights [1,1,1,1]
   }
   n7 = "Constraint - deform_hip_geo_base_b_r (PPPPP)"!(
    ComponentPose=@n6
    Alpha=1
    ConstraintWeights_0=0.8966
    ConstraintWeights_1:float=0.0684
    ConstraintWeights_2:float=0.0188
    ConstraintWeights_3:float=0.0103
    ConstraintWeights_4:float=0.0059
   ) -> (Pose) {
    alphaBoolBlend {blendOption=Linear}
    boneToModify {boneName=deform_hip_geo_base_b_r}
    constraintSetup [
     {targetBone {boneName=pelvis}}
     {targetBone {boneName=spine_02}}
     {targetBone {boneName=thigh_twist_01_r}}
     {targetBone {boneName=spine_01}}
     {targetBone {boneName=deform_glute_r}}
    ]
    constraintWeights [1,1,1,1,1]
   }
   n8 = "Constraint - deform_hip_geo_base_c_r (PPPPP)"!(
    ComponentPose=@n7
    Alpha=1
    ConstraintWeights_0=0.7491
    ConstraintWeights_1:float=0.159
    ConstraintWeights_2:float=0.0648
    ConstraintWeights_3:float=0.0138
    ConstraintWeights_4:float=0.0133
   ) -> (Pose) {
    alphaBoolBlend {blendOption=Linear}
    boneToModify {boneName=deform_hip_geo_base_c_r}
    constraintSetup [
     {targetBone {boneName=pelvis}}
     {targetBone {boneName=spine_02}}
     {targetBone {boneName=spine_01}}
     {targetBone {boneName=spine_04}}
     {targetBone {boneName=spine_03}}
    ]
    constraintWeights [1,1,1,1,1]
   }
   n9 = "Transform (Modify) Bone - Bone: deform_kneepad_adjust_r"(ComponentPose=@n8, Alpha=1, Translation="0.212,-1.095,1.901", Rotation="-6.202,-0.782,-16.910", Scale="1.000,1.000,1.155") {
    rotationMode=BMM_Additive
    scaleMode=BMM_Additive
    translationMode=BMM_Additive
    alphaBoolBlend {blendInTime=1, blendOption=Cubic, blendOutTime=1}
    alphaScaleBiasClamp {bInterpResult=true, bMapRange=true, inRange {max=110}}
    boneToModify {boneName=deform_kneepad_adjust_r}
    rotation=<rot -6.202 -0.782 -16.91>
    scale {z=1.155}
    translation=<0.212 -1.095 1.901>
   }
   Output Pose(Result=Component To Local(ComponentPose=@n9)) {layerGroup=None}
  ]}
 EventGraph {nodes [
   Event BlueprintUpdateAnimation()
   > n1 = Sequence()
   > n2 = Branch(Condition=Get bIsCrouching())
   n3 = GetOwnerSkeletalMeshComponent()
   n4 = BreakRotator(InRot=GetDeltaTransformFromRefPose!(self=@n3, BoneName=calf_r, BaseName=thigh_r) -> (ReturnValue_Location:Vector, ReturnValue_Rotation:Rotator, ReturnValue_Scale:Vector).ReturnValue_Rotation)
   > Set CrouchingKneeR(CrouchingKneeR:float=@n4.Yaw) -> (Output_Get:float)
   Set CrouchingKneeR(execute=@n2.else, CrouchingKneeR=0)
   n5 = Branch(execute=@n1.then_1, Condition=Get bIsSliding())
   > Set SlidingKneeL(SlidingKneeL:float=BreakRotator(InRot=GetDeltaTransformFromRefPose(self=@n3, BoneName=calf_l, BaseName=thigh_l).ReturnValue_Rotation).Yaw) -> (Output_Get:float)
   > Set SlidingKneeR(SlidingKneeR:float=@n4.Yaw) -> (Output_Get:float)
   Set SlidingKneeL(execute=@n5.else, SlidingKneeL=0)
   > Set SlidingKneeR(SlidingKneeR=0)
  ]}
]
variables [
 CrouchingKneeR {type=number}
 Glute Transform {type=Transform, default {location=<-6 2 5>, rotation=<rot 0 0 0>, scale=<0.5 1.5 1.5>}}
 SlidingKneeL {type=number}
 SlidingKneeR {type=number}
]
