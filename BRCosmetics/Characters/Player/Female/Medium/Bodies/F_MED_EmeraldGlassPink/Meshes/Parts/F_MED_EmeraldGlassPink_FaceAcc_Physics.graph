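# Physics asset for the FaceAcc part of the EmeraldGlassPink medium female player body, covering its hair: bangs, left and right strands, and layered back hair.
# Small master-joint capsules under the head anchor the dynamic capsule chains, with side upper layers branching from the first center back joint.
# Dynamic hair joints have limited swing and near-locked twist.
# Head, neck, upper spine and clavicle shapes form the surrounding body, linked by free angular constraints.
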PhysicsAsset /BRCosmetics/Characters/Player/Female/Medium/Bodies/F_MED_EmeraldGlassPink/Meshes/Parts/F_MED_EmeraldGlassPink_FaceAcc_Physics
bodies {
 clavicle_l [{shapeType=Capsule, length=8.5, radius=6.005, shapeName=clavicle_l_capsule, center=<5.6354 0 -1.9085>, rotation=<rot -0.0002 -90 -80>}]
 clavicle_r [{shapeType=Capsule, length=8.5, radius=6.005, shapeName=clavicle_r_capsule, center=<-6.4674 0 2.0475>, rotation=<rot -0.0002 -90 -80>}]
 dyn_L_hair_strand_1_jnt [{
   shapeType=Capsule
   length=2
   radius=2.5
   shapeName=dyn_L_hair_strand_1_jnt_capsule
   center=<2.5737 0 -0.4538>
   rotation=<rot -80 -179.9998 -160>
  }]
 dyn_L_hair_strand_2_jnt [{
   shapeType=Capsule
   length=2
   radius=2.5
   shapeName=dyn_L_hair_strand_2_jnt_capsule
   center=<2.5737 0 -0.4538>
   rotation=<rot -80 -179.9998 -160>
  }]
 dyn_L_hair_strand_3_jnt [{
   shapeType=Capsule
   length=2
   radius=2.5
   shapeName=dyn_L_hair_strand_3_jnt_capsule
   center=<2.5737 0 -0.4538>
   rotation=<rot -80 -179.9997 -160>
  }]
 dyn_R_hair_strand_1_jnt [{
   shapeType=Capsule
   length=2
   radius=2.5
   shapeName=dyn_R_hair_strand_1_jnt_capsule
   center=<-2.4389 0 0>
   rotation=<rot 90 -160.5289 -160.5292>
  }]
 dyn_R_hair_strand_2_jnt [{shapeType=Capsule, length=2, radius=2.5, shapeName=dyn_R_hair_strand_2_jnt_capsule, center=<-2.4389 0 0>, rotation=<rot 90 0 -0.0004>}]
 dyn_R_hair_strand_3_jnt [{
   shapeType=Capsule
   length=2
   radius=2.5
   shapeName=dyn_R_hair_strand_3_jnt_capsule
   center=<-2.4389 0 0>
   rotation=<rot 90 -179.9986 -179.999>
  }]
 dyn_bangs_jnt [{shapeType=Capsule, length=3, radius=3.5, shapeName=dyn_bangs_jnt_capsule, center=<1.7483 2.1515 0>}]
 dyn_hair_C_upper_layer_1_jnt [{shapeType=Capsule, length=4.5, radius=4.75, shapeName=dyn_hair_C_upper_layer_1_jnt_capsule, center=<4.7322 -3.2299 0>}]
 dyn_hair_C_upper_layer_2_jnt [{shapeType=Capsule, length=4.5, radius=4.75, shapeName=dyn_hair_C_upper_layer_2_jnt_capsule, center=<4.7322 -3.2299 0>}]
 dyn_hair_L_upper_layer_1_jnt [{
   shapeType=Capsule
   length=10
   radius=3
   shapeName=dyn_hair_R_upper_layer_1_jnt1_capsule
   center=<-6.3149 -1.469 1.8677>
   rotation=<rot 90 0 -20.0003>
  }]
 dyn_hair_R_upper_layer_1_jnt1 [{
   shapeType=Capsule
   length=10
   radius=3
   shapeName=dyn_hair_R_upper_layer_1_jnt1_capsule
   center=<6.4346 1.7558 -1.5672>
   rotation=<rot 90 0 -20.0003>
  }]
 head [
  {shapeType=Sphere, radius=10.75, shapeName=head_sphere, center=<7.158 -3.7671 0>}
  {shapeType=Capsule, length=2.75, radius=10, center=<5.8315 0 0>, rotation=<rot -90 0.0001 -0.0001>}
 ]
 master_L_hair_strand_jnt [{shapeType=Capsule, length=0.5, radius=0.5, shapeName=master_L_hair_strand_jnt_capsule}]
 master_R_hair_strand_jnt [{shapeType=Capsule, length=0.5, radius=0.5, shapeName=master_R_hair_strand_jnt_capsule}]
 master_bangs_jnt [{shapeType=Capsule, length=0.5, radius=0.5, shapeName=master_bangs_jnt_capsule}]
 master_hair_back_jnt [{shapeType=Capsule, length=0.5, radius=0.5, shapeName=master_hair_back_jnt_capsule}]
 master_hair_jnt [{shapeType=Capsule, length=0.5, radius=0.5}]
 neck_01 [{shapeType=Capsule, length=2.75, radius=3, shapeName=neck_01_capsule, rotation=<rot -90 -0.0002 0>}]
 neck_02 [{shapeType=Capsule, length=2.75, radius=3, shapeName=neck_02_capsule, rotation=<rot -90 180 179.9998>}]
 spine_05 [{shapeType=Capsule, length=10.75, radius=10.005, shapeName=spine_05_capsule, center=<-3.0995 3.0877 0>, rotation=<rot 0 -90 0>}]
}
constraints [
 {bone1Name=master_hair_jnt, bone2Name=head, swing1LimitDegrees=45, swing2LimitDegrees=45, twistLimitDegrees=45}
 {
  bone1Name=head
  bone2Name=neck_02
  swing1LimitDegrees=45
  swing1Motion=Free
  swing2LimitDegrees=45
  swing2Motion=Free
  twistLimitDegrees=45
  twistMotion=Free
 }
 {
  bone1Name=neck_02
  bone2Name=neck_01
  swing1LimitDegrees=45
  swing1Motion=Free
  swing2LimitDegrees=45
  swing2Motion=Free
  twistLimitDegrees=45
  twistMotion=Free
 }
 {
  bone1Name=master_L_hair_strand_jnt
  bone2Name=master_hair_jnt
  swing1LimitDegrees=45
  swing1Motion=Free
  swing2LimitDegrees=45
  swing2Motion=Free
  twistLimitDegrees=45
  twistMotion=Free
 }
 {bone1Name=dyn_L_hair_strand_1_jnt, bone2Name=master_L_hair_strand_jnt, swing1LimitDegrees=45, swing2LimitDegrees=45, twistLimitDegrees=1}
 {bone1Name=dyn_L_hair_strand_2_jnt, bone2Name=dyn_L_hair_strand_1_jnt, swing1LimitDegrees=45, swing2LimitDegrees=45, twistLimitDegrees=1}
 {bone1Name=dyn_L_hair_strand_3_jnt, bone2Name=dyn_L_hair_strand_2_jnt, swing1LimitDegrees=45, swing2LimitDegrees=45, twistLimitDegrees=1}
 {
  bone1Name=master_R_hair_strand_jnt
  bone2Name=master_hair_jnt
  swing1LimitDegrees=45
  swing1Motion=Free
  swing2LimitDegrees=45
  swing2Motion=Free
  twistLimitDegrees=45
  twistMotion=Free
 }
 {bone1Name=dyn_R_hair_strand_1_jnt, bone2Name=master_R_hair_strand_jnt, swing1LimitDegrees=45, swing2LimitDegrees=45, twistLimitDegrees=1}
 {bone1Name=dyn_R_hair_strand_2_jnt, bone2Name=dyn_R_hair_strand_1_jnt, swing1LimitDegrees=45, swing2LimitDegrees=45, twistLimitDegrees=1}
 {bone1Name=dyn_R_hair_strand_3_jnt, bone2Name=dyn_R_hair_strand_2_jnt, swing1LimitDegrees=45, swing2LimitDegrees=45, twistLimitDegrees=1}
 {
  bone1Name=master_bangs_jnt
  bone2Name=master_hair_jnt
  swing1LimitDegrees=45
  swing1Motion=Free
  swing2LimitDegrees=45
  swing2Motion=Free
  twistLimitDegrees=45
  twistMotion=Free
 }
 {bone1Name=dyn_bangs_jnt, bone2Name=master_bangs_jnt, swing1LimitDegrees=2, swing2LimitDegrees=45, twistLimitDegrees=1}
 {
  bone1Name=master_hair_back_jnt
  bone2Name=master_hair_jnt
  swing1LimitDegrees=45
  swing1Motion=Free
  swing2LimitDegrees=45
  swing2Motion=Free
  twistLimitDegrees=45
  twistMotion=Free
 }
 {bone1Name=dyn_hair_C_upper_layer_1_jnt, bone2Name=master_hair_back_jnt, swing1LimitDegrees=55, swing2LimitDegrees=12, twistLimitDegrees=1}
 {
  bone1Name=dyn_hair_C_upper_layer_2_jnt
  bone2Name=dyn_hair_C_upper_layer_1_jnt
  swing1LimitDegrees=25
  swing2LimitDegrees=12
  twistLimitDegrees=1
 }
 {
  bone1Name=dyn_hair_L_upper_layer_1_jnt
  bone2Name=dyn_hair_C_upper_layer_1_jnt
  swing1LimitDegrees=35
  swing2LimitDegrees=12
  twistLimitDegrees=1
 }
 {
  bone1Name=dyn_hair_R_upper_layer_1_jnt1
  bone2Name=dyn_hair_C_upper_layer_1_jnt
  swing1LimitDegrees=35
  swing2LimitDegrees=12
  twistLimitDegrees=1
 }
 {
  bone1Name=neck_01
  bone2Name=spine_05
  swing1LimitDegrees=45
  swing1Motion=Free
  swing2LimitDegrees=45
  swing2Motion=Free
  twistLimitDegrees=45
  twistMotion=Free
 }
 {
  bone1Name=clavicle_l
  bone2Name=spine_05
  swing1LimitDegrees=45
  swing1Motion=Free
  swing2LimitDegrees=45
  swing2Motion=Free
  twistLimitDegrees=45
  twistMotion=Free
 }
 {
  bone1Name=clavicle_r
  bone2Name=spine_05
  swing1LimitDegrees=45
  swing1Motion=Free
  swing2LimitDegrees=45
  swing2Motion=Free
  twistLimitDegrees=45
  twistMotion=Free
 }
]
