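# Post-body AnimBP for the EmeraldGlassPink face accessory, adding hair physics and correcting the right hair strand while targeting.
# It copies the attached parent's pose, then runs a RigidBody simulation in head-bone space with custom gravity.
# Two additive pitch rotations follow on the strand's master bone with interpolated alphas: one when targeting with head yaw past a threshold relative to spine_05, another whenever targeting.
# Both alphas are set each animation update.
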AnimBlueprint /BRCosmetics/Characters/Player/Female/Medium/Bodies/F_MED_EmeraldGlassPink/Meshes/Parts/F_MED_EmeraldGlassPink_FaceAcc_AnimBP
target_skeleton=&/F_MED_EmeraldGlassPink_FaceAcc_Skeleton
graphs [
 AnimGraph {nodes [
   n1 = RigidBody(ComponentPose=Local To Component(LocalPose=Copy Pose From Mesh() {bUseAttachedParent=true})) {
    bClampLinearTranslationLimitToRefPose=true
    bOverrideWorldGravity=true
    overridePhysicsAsset=&/F_MED_EmeraldGlassPink_FaceAcc_Physics
    simulationSpace=BaseBoneSpace
    alphaBoolBlend {blendOption=Linear}
    baseBoneRef {boneName=head}
    componentAppliedLinearAccClamp=<6000 6000 6000>
    componentLinearAccScale=<0.35 0.2 0.3>
    componentLinearVelScale=<0.35 0.2 0.3>
    overrideWorldGravity {z=-980}
    simSpaceSettings {maxAngularAcceleration=1000, maxAngularVelocity=3, maxLinearAcceleration=250, maxLinearVelocity=500, velocityScaleZ=0, worldAlpha=1}
   }
   n2 = "Transform (Modify) Bone - Bone: master_R_hair_strand_jnt"(ComponentPose=@n1, Alpha=Get Head Back/ADS() -> (Head Back/ADS:double), Translation="0,0,0", Rotation="-12,0,0", Scale="1,1,1") {
    rotationMode=BMM_Additive
    rotationSpace=BCS_BoneSpace
    alphaBoolBlend {blendOption=Linear}
    alphaScaleBiasClamp {bInterpResult=true, interpSpeedDecreasing=2, interpSpeedIncreasing=5}
    boneToModify {boneName=master_R_hair_strand_jnt}
    rotation {pitch=-12}
   }
   n3 = "Transform (Modify) Bone - Bone: master_R_hair_strand_jnt"(ComponentPose=@n2, Alpha=Get ADS() -> (ADS:double), Translation="0,0,0", Rotation="-10,0,0", Scale="1,1,1") {
    rotationMode=BMM_Additive
    rotationSpace=BCS_BoneSpace
    alphaBoolBlend {blendOption=Linear}
    alphaScaleBiasClamp {bInterpResult=true, interpSpeedDecreasing=2, interpSpeedIncreasing=5}
    boneToModify {boneName=master_R_hair_strand_jnt}
    rotation {pitch=-10}
   }
   Output Pose(Result=Component To Local(ComponentPose=@n3)) {layerGroup=None}
  ]}
 EventGraph {nodes [
   Event BlueprintUpdateAnimation()
   n1 = BreakTransform(InTransform=GetDeltaTransformFromRefPose(self=GetOwnerSkeletalMeshComponent(), BoneName=head, BaseName=spine_05))
   n2 = "float >= float"(A=MapRangeClamped(Value=BreakRotator(InRot=@n1.Rotation).Yaw, InRangeA=-10, InRangeB=-55), B=0.5)
   n3 = Get bIsTargeting()
   > Set Head Back/ADS(Head Back/ADS:double="To Float (Boolean)"(InBool=AND Boolean(A=@n2, B=@n3))) -> (Output_Get:double)
   > Set ADS(ADS:double="To Float (Boolean)"(InBool=Get bIsTargeting())) -> (Output_Get:double)
  ]}
]
variables [
 ADS {type=number}
 Head Back/ADS {type=number}
]
